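# First-person fire sound cue for the NoScope sniper rifle in the Figment Athena content.
# A Random node picks between two equally weighted close-range fire waves and feeds a root Modulator, whose pitch is pinned so it adds no pitch variation.
# It uses first-person projectile weapon attenuation and the first-person ranged fire sound class.
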SoundCue /Figment_Sniper_NoScope_Athena/Gameplay/Sounds/FN03_Weapon_RFL_Sniper_NoScope_Athena_1P_Cue
root=Modulator_0
nodes [
 {id=Modulator_0, children [Random_0], properties {pitchMax=1, pitchMin=1}}
 {id=Random_0, children [WavePlayer_0,WavePlayer_1], properties {weights [1,1]}}
 {id=WavePlayer_0, properties {soundWaveAssetPtr=/Game/Athena/Sounds/Weapons/Sniper_NoScope/Sniper_NoScope_Fire_Close_02}}
 {id=WavePlayer_1, properties {soundWaveAssetPtr=/Game/Athena/Sounds/Weapons/Sniper_NoScope/Sniper_NoScope_Fire_Close_01}}
]
properties {
 attenuationSettings=&Fort_Weapon_Projectile_FirstPerson_Athena_Att
 maxDistance=2097152
 soundClassObject=&Fort_WeaponRanged_Fire_1P_PSM
 volumeMultiplier=1
}
